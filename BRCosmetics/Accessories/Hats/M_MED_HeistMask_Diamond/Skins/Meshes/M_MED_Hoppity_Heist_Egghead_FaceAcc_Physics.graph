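# Physics asset for the Hoppity Heist Egghead face accessory, a cosmetic skin filed under the Diamond heist mask, covering its hat and ribbon bones.
# Single capsule bodies sit on the head, the dynamic hat bone and two dynamic ribbon bones.
# Each ribbon is constrained to the hat with narrow swing limits and locked twist, while no constraint links the hat to the head.
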PhysicsAsset /BRCosmetics/Accessories/Hats/M_MED_HeistMask_Diamond/Skins/Meshes/M_MED_Hoppity_Heist_Egghead_FaceAcc_Physics
bodies {
 dyn_hat [{shapeType=Capsule, length=8.9128, radius=4.6115, center=<2.3753 0.3495 -5.7417>, rotation=<rot 78.2358 -89.9992 0.0003>}]
 dyn_ribbon_01 [{shapeType=Capsule, length=2.009, radius=1.186, center=<1.0917 -1.2368 0.3324>, rotation=<rot -43.1295 -137.7564 -98.3354>}]
 dyn_ribbon_02 [{shapeType=Capsule, length=1.1489, radius=1.25, center=<1.0776 -1.3455 -0.6267>, rotation=<rot 37.806 -121.3691 -56.0709>}]
 head [{shapeType=Capsule, length=0.5337, radius=1.1485, center=<-0.5263 0.0426 -0.0649>, rotation=<rot 0 -110.2322 -76.0522>}]
}
constraints [
 {bone1Name=dyn_ribbon_01, bone2Name=dyn_hat, swing1LimitDegrees=15, swing2LimitDegrees=10, twistLimitDegrees=1, twistMotion=Locked}
 {bone1Name=dyn_ribbon_02, bone2Name=dyn_hat, swing1LimitDegrees=15, swing2LimitDegrees=10, twistLimitDegrees=1, twistMotion=Locked}
]
